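# Material for cube monster-shard ingredient items, with animated glowing runes and drifting chunks.
# Vertex colors pick sine-driven local offsets, which combine with axis rotation in World Position Offset, with corrected normals.
# Emissive layers flickering, gradient-panned rune masks over a pulsing sphere-mask glow that follows chunk motion.
# An IsElectrical static switch swaps to emissive and opacity built from HDR color, fresnel, scrolling noise and camera distance, with spline-thickened vertices.
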Material /CorruptionGameplay/Gameplay/Items/Ingredients/MonsterShards/Assets/Materials/M_Gadget_Cube_Shard
graph {
 properties {bUseMaterialAttributes=true, pixelDepthOffsetMode=PDOM_Legacy, usedShadingModels=Default Lit}
 nodes [
  n1 = NamedRerouteUsage()
  n2 = ScalarParameter() {defaultValue=0.5, parameterName=EL-Width}
  n3 = LinearInterpolate(A=Multiply(A=@n2) {constB=15}, B=@n2, Alpha=VertexColor().A)
  Add(
   A=@n1
   B=MaterialFunctionCall(WidthBase=@n3, WidthTip=@n3, WorldPosition=@n1, Expand U or V UV Channel=StaticBool()) {materialFunction=&SplineThicken}.WorldPositionOffset
  )
  Constant3Vector() {constant=<rgba 0.5 0.7 3.5 0>}
  Constant3Vector() {constant=<rgba 0 0 3 0>}
  Constant() {r=3.5}
  Constant() {r=0.07}
  Constant() {r=700}
  Constant() {r=0.5}
  Constant()
  Constant() {r=0.1}
  Constant() {r=1}
  n4 = TextureSampleParameter2D() {
   group=Emissive
   parameterName=EmissiveMasks
   texture=/CorruptionGameplay/Gameplay/Items/Ingredients/MonsterShards/Assets/Textures/T_Item_Cube_Shard_M
  }
  n5 = Add(A=Multiply(A=Time(), B=ScalarParameter() {defaultValue=1, parameterName=Cube Rune Flicker Speed}), B=@n4.G)
  n6 = Multiply(A=Sine(None=@n5) {period=0.22}, B=Sine(None=@n5) {period=0.11})
  n7 = Sine(None=@n5) {period=0.65}
  n8 = Multiply(A=Clamp(None=Add(A=@n6, B=@n7)) {maxDefault=2}) {constB=2}
  n9 = ConstantBiasScale(None=Sine(None=@n5) {period=3})
  n10 = Divide(A=ComponentMask(None=WorldPosition().XYZ) {b=true}, B=ScalarParameter() {defaultValue=50, parameterName=GradientTile}) {constB=200}
  n11 = Time()
  n12 = ScalarParameter() {defaultValue=1, parameterName=Random}
  n13 = Add(A=@n11, B=@n12)
  n14 = Multiply(A=ScalarParameter() {defaultValue=-0.27, parameterName=GradientPanSpeed}, B=@n13) {constB=-0.2}
  n15 = Multiply(
   A=LinearInterpolate(A=Multiply(A=@n8, B=@n9), Alpha=Power(Base=@n9) {constExponent=3})
   B=TextureSample(UVs=Add(A=@n10, B=@n14)) {texture=/Game/Global/Textures/Default/Masks_and_Gradients/T_VertGradient_02_sRGB}.G
  )
  n16 = Multiply(A=ScalarParameter() {defaultValue=0.1, parameterName=Cube Rune Dark Amt}, B=@n4.R)
  n17 = VectorParameter() {group=Emissive, parameterName=EmissiveEdgeColor, defaultValue=<rgba 1.0742 0.1719 3 0>}
  n18 = LinearInterpolate(A=@n17.RGB, B=VectorParameter() {group=Emissive, parameterName=EmissiveColor, defaultValue=<rgba 4.6976 0.1302 5 0>}.RGB, Alpha=@n4.R)
  n19 = TextureSampleParameter2D() {
   desc="R = SpecMask G = Metallic B = Roughness"
   group=Specular
   parameterName=SpecularMasks
   samplerSource=SSM_Wrap_WorldGroupSettings
   samplerType=SAMPLERTYPE_Masks
   texture=/CorruptionGameplay/Gameplay/Items/Ingredients/MonsterShards/Assets/Textures/T_Item_Cube_Shard_S
  }
  n20 = VertexColor()
  n21 = Multiply(A=@n20.R, B=Sine(None=@n13) {period=3.3})
  n22 = Multiply(A=ScalarParameter() {defaultValue=2, parameterName=CubeChunkAmt1}, B=Sine(None=@n13) {period=-2})
  n23 = Multiply(
   A=LinearInterpolate(A=LinearInterpolate(A=@n21, B=@n22, Alpha=@n20.G), B=Sine(None=@n13) {period=4}, Alpha=@n20.B)
   B=Transform(None=Constant3Vector() {constant=<rgba 1.1 0.7 3.5 0>}) {transformSourceType=TRANSFORMSOURCE_Local}
  )
  n24 = Transform(None=VectorParameter() {parameterName=PivotOffset}.RGB) {transformSourceType=TRANSFORMSOURCE_Local}
  n25 = SphereMask(
   A=WorldPosition().XYZ
   B=Add(A=ActorPositionWS(), B=Add(A=@n23, B=@n24))
   Radius=ScalarParameter() {defaultValue=35, parameterName=GlowRadius}
   Hardness=ScalarParameter() {defaultValue=0.3, parameterName=GlowHardness}
  ) {hardnessPercent=30}
  n26 = ScalarParameter() {defaultValue=2, parameterName=GlowPower}
  n27 = Multiply(
   A=Multiply(A=LinearInterpolate(Alpha=@n19.B) {constA=1.5, constB=0.4}, B=Power(Base=@n25, Exp=@n26))
   B=ScalarParameter() {defaultValue=1, parameterName=Brightness}
  )
  n28 = LinearInterpolate(
   A=ScalarParameter() {defaultValue=0.15, parameterName=InnerGlowPulseLerpA}
   B=ScalarParameter() {defaultValue=0.35, parameterName=InnerGlowPulseLerpB}
   Alpha=MaterialFunctionCall(Speed=ScalarParameter() {defaultValue=0.35, parameterName=InnerGlowPulseSpeed}) {materialFunction=&DebugTimeSine}.Result
  ) {constA=0.05, constB=0.2}
  n29 = Add(A=Multiply(A=Add(A=@n15, B=@n16), B=Multiply(A=@n4.R, B=@n18)), B=Multiply(A=@n17.RGB, B=Multiply(A=@n27, B=@n28) {constB=0.2}))
  Multiply(A=StaticSwitchParameter(True=Constant() {r=1}, False=ParticleColor().RGB) {parameterName=ApplytoSkeletal}, B=@n29)
  n30 = TextureSampleParameter2D() {
   parameterName=Normals
   samplerType=SAMPLERTYPE_Normal
   texture=/CorruptionGameplay/Gameplay/Items/Ingredients/MonsterShards/Assets/Textures/T_Item_Cube_Shard_N
  }
  Multiply(A=@n30.RGB)
  Multiply(A=ScalarParameter() {defaultValue=1, parameterName=CubeChunkSpeed}, B=@n13)
  n31 = Normalize(VectorInput=Transform(None=Constant3Vector() {constant=<rgba 0 0 1 0>}) {transformSourceType=TRANSFORMSOURCE_Local})
  n32 = Multiply(A=@n11, B=Multiply(A=@n12, B=ScalarParameter() {defaultValue=0.07, parameterName=RotateSpeed}) {constB=0.2}) {constB=0.1}
  n33 = Add(A=@n23, B=Add(A=@n24, B=ActorPositionWS()))
  n34 = WorldPosition()
  NamedRerouteDeclaration(None=Add(A=@n23, B=RotateAboutAxis(NormalizedRotationAxis=@n31, RotationAngle=@n32, PivotPoint=@n33, Position=@n34.XYZ))) {name=WPO, nodeColor=<rgba 0 1 0.7891 1>}
  n35 = MaterialFunctionCall(Rotation Axis=@n31, Rotation Angle=@n32) {materialFunction=&FixRotateAboutAxisNormals}
  NamedRerouteDeclaration(None=@n35.New Tangent Space Vertex Normal) {name=FixedRAAN, nodeColor=<rgba 0 1 0.7891 1>}
  Sine(None=@n13) {period=3.3}
  n36 = VectorParameter() {parameterName=EL-Color, defaultValue=<rgba 0.6066 0 1 1>}
  n37 = ScalarParameter() {defaultValue=10, parameterName=EL-Brightness}
  n38 = MaterialFunctionCall(LDR Input=@n36.RGB, HDR Dynamic Range=@n37, HDR Tint=Multiply(A=@n36.RGB, B=@n37), HDR Intensity Multiplier=@n37) {materialFunction=&DeriveHDRfromLDR}
  n39 = VertexColor()
  n40 = Multiply(A=MaterialFunctionCall(Float2=TextureCoordinate()) {materialFunction=&BreakOutFloat2Components}.G) {constB=0.5}
  n41 = Power(Base=Saturate(None=Sine(None=@n40))) {constExponent=3}
  n42 = MaterialFunctionCall(
   Normal Vector=Transform(None=NamedRerouteUsage())
   Invert Fresnel=StaticBool() {value=true}
   Power=ScalarParameter() {defaultValue=3.5, group=Emissive, parameterName=EL-Emissive_FresPow}
  ) {materialFunction=&Fresnel_Function}
  n43 = Multiply(A=DepthFade(Opacity=Divide(A=@n41) {constB=4}, FadeDistance=Constant() {r=20}), B=@n42.Result)
  n44 = Saturate(None=Multiply(A=@n42.Result, B=ScalarParameter() {defaultValue=1.5, group=Emissive, parameterName=EL-Emissive_FresMod}))
  n45 = VectorParameter() {parameterName=EL-EmissiveNoise, defaultValue=<rgba 0.5 2 0.5 1>}
  n46 = Multiply(A=TextureCoordinate(), B=AppendVector(A=@n45.R, B=@n45.R))
  n47 = Multiply(A=Time(), B=@n45.B)
  n48 = LinearInterpolate(
   A=ScalarParameter() {defaultValue=-0.3, group=Emissive, parameterName=EL-Emissive_NoiseLow}
   Alpha=TextureSample(UVs=Add(A=@n46, B=@n47)) {samplerType=SAMPLERTYPE_LinearGrayscale, texture=/Game/Global/Textures/Default/Noise/T_SwirlNoise_N_height}.R
  )
  n49 = SphereMask(
   A=CameraPositionWS()
   B=WorldPosition().XYZ
   Radius=ScalarParameter() {defaultValue=1400, group=Emissive, parameterName=EL-Emissive-Camera-Rad}
   Hardness=ScalarParameter() {defaultValue=0.25, group=Emissive, parameterName=EL-Emissive-Camera-Hard}
  ) {attenuationRadius=64, hardnessPercent=20}
  n50 = Multiply(
   A=@n41
   B=LinearInterpolate(
    A=ScalarParameter() {defaultValue=0.5, group=Emissive, parameterName=EL-Emissive-CameraFarFlat}
    B=MaterialFunctionCall(Transition Phase=@n44, Height Texture=@n48) {materialFunction=&HeightLerp}.Lerp Alpha No Contrast
    Alpha=Saturate(None=@n49)
   )
  )
  n51 = VectorParameter() {parameterName=BasicDiffuseColor, defaultValue=<rgba 0.0263 0.005 0.05 1>}
  n52 = SetMaterialAttributes!(
   MaterialAttributes
   Base Color=@n51.RGB
   Metallic=@n19.G
   Specular=@n19.G
   Roughness=@n19.G
   Emissive Color=@n29
   Normal=MaterialFunctionCall(BaseNormal=@n35.New Tangent Space Vertex Normal, AdditionalNormal=@n30.RGB) {materialFunction=&BlendAngleCorrectedNormals}.Result
   World Position Offset=NamedRerouteUsage()
  ) {attributeSetTypes [
    #1
    #2
    #3
    #4
    #5
    #6
    #7
   ]}
  StaticSwitchParameter(
   True=SetMaterialAttributes!(
    MaterialAttributes
    Emissive Color=LinearInterpolate(A=@n36.RGB, B=@n38.RemappedRange, Alpha=@n39.A) {desc=emissive}
    Opacity=LinearInterpolate(A=@n43, B=@n50, Alpha=@n39.A) {desc=opacity}
    World Position Offset=NamedRerouteUsage()
   ) {attributeSetTypes [#5, #8, #7]}
   False=@n52
  ) {parameterName=IsElectrical}
  StaticSwitchParameter(True=TextureSampleParameter2D() {parameterName=Diffuse}.RGB, False=@n51.RGB) {parameterName=UseDiffuseTexture}
 ]
}
